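# Physics asset for the tailgate damage mesh of the Valet pickup truck, giving it simple collision and a single-axis joint.
# Two capsule bodies cover the body and trunk bones, with the trunk capsule offset behind and above its bone origin.
# One constraint joins trunk to body with twist and swing1 locked, leaving only swing2, which has a small limit angle, as a hinge-like axis.
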PhysicsAsset /Valet/Pickup_Truck/Meshes/Pickup_Truck_Veh_Dyn_dmg_tailgate
bodies {
 body [{shapeType=Capsule, length=77.8438, radius=40, center=<-3.643 0.0005 -0.0002>, rotation=<rot 0 0 89.9997>}]
 trunk [{shapeType=Capsule, length=115.1562, radius=23.8125, center=<-24.295 0.0008 48.5954>, rotation=<rot 0 0 90>}]
}
constraints [{bone1Name=trunk, bone2Name=body, swing1LimitDegrees=45, swing1Motion=Locked, swing2LimitDegrees=10, twistLimitDegrees=45, twistMotion=Locked}]
