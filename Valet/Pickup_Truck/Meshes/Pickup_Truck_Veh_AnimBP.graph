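# Animation Blueprint for the pickup truck vehicle, posing axles, wheel steering and spin, and the steering wheel with chained bone transforms on a reference pose.
# RigidBody nodes using per-part physics assets let the tailgate and each mirror dangle, with alphas switched on as health ratio drops past separate thresholds.
# Each update also reads the tire-mod flag from the Dagwood vehicle, which shifts the front knuckles and rear wheel bones sideways.
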AnimBlueprint /Valet/Pickup_Truck/Meshes/Pickup_Truck_Veh_AnimBP
target_skeleton=&/Pickup_Truck_Veh_Skeleton
graphs [
 AnimGraph {nodes [
   n1 = Get AxlePositionF() -> (AxlePositionF:Vector)
   n2 = Get AxleRotationF() -> (AxleRotationF:Rotator)
   n3 = "Transform (Modify) Bone - Bone: axle_pivot_fr_l"(ComponentPose=Local To Component(LocalPose=Local Space Ref Pose()), Alpha=1, Translation=@n1, Rotation=@n2, Scale="1,1,1") {
    rotationMode=BMM_Replace
    rotationSpace=BCS_BoneSpace
    translationMode=BMM_Replace
    translationSpace=BCS_BoneSpace
    alphaBoolBlend {blendOption=Linear}
    boneToModify {boneName=axle_pivot_fr_l}
   }
   n4 = "Transform (Modify) Bone - Bone: axle_pivot_fr_r"(ComponentPose=@n3, Alpha=1, Translation=@n1, Rotation=@n2, Scale="1,1,1") {
    rotationMode=BMM_Replace
    rotationSpace=BCS_BoneSpace
    translationMode=BMM_Replace
    translationSpace=BCS_BoneSpace
    alphaBoolBlend {blendOption=Linear}
    boneToModify {boneName=axle_pivot_fr_r}
   }
   n5 = "Transform (Modify) Bone - Bone: rear_differential"(
    ComponentPose=@n4
    Alpha=1
    Translation=Get AxlePositionB() -> (AxlePositionB:Vector)
    Rotation=Get AxleRotationB() -> (AxleRotationB:Rotator)
    Scale="1,1,1"
   ) {
    rotationMode=BMM_Replace
    rotationSpace=BCS_BoneSpace
    translationMode=BMM_Replace
    translationSpace=BCS_BoneSpace
    alphaBoolBlend {blendOption=Linear}
    boneToModify {boneName=rear_differential}
   }
   n6 = "Transform (Modify) Bone - Bone: wheel_steering_fr_r"(ComponentPose=@n5, Alpha=1, Translation="0,0,0", Rotation=Get WheelRotationFR() -> (WheelRotationFR:Rotator), Scale="1,1,1") {
    rotationMode=BMM_Replace
    rotationSpace=BCS_BoneSpace
    translationSpace=BCS_BoneSpace
    alphaBoolBlend {blendOption=Linear}
    boneToModify {boneName=wheel_steering_fr_r}
   }
   n7 = "Transform (Modify) Bone - Bone: wheel_steering_fr_l"(ComponentPose=@n6, Alpha=1, Translation="0,0,0", Rotation=Get WheelRotationFL() -> (WheelRotationFL:Rotator), Scale="1,1,1") {
    rotationMode=BMM_Replace
    rotationSpace=BCS_BoneSpace
    translationSpace=BCS_BoneSpace
    alphaBoolBlend {blendOption=Linear}
    boneToModify {boneName=wheel_steering_fr_l}
   }
   n8 = "Transform (Modify) Bone - Bone: wheel_tilted_fr_r"(ComponentPose=@n7, Alpha=1, Translation="0,0,0", Rotation=Get WheelSpinFR() -> (WheelSpinFR:Rotator), Scale="1,1,1") {rotationMode=BMM_Replace, rotationSpace=BCS_BoneSpace, alphaBoolBlend {blendOption=Linear}, boneToModify {boneName=wheel_tilted_fr_r}}
   n9 = "Transform (Modify) Bone - Bone: wheel_tilted_fr_l"(ComponentPose=@n8, Alpha=1, Translation="0,0,0", Rotation=Get WheelSpinFL() -> (WheelSpinFL:Rotator), Scale="1,1,1") {rotationMode=BMM_Replace, rotationSpace=BCS_BoneSpace, alphaBoolBlend {blendOption=Linear}, boneToModify {boneName=wheel_tilted_fr_l}}
   n10 = "Transform (Modify) Bone - Bone: wheel_tilted_bk_r"(ComponentPose=@n9, Alpha=1, Translation="0,0,0", Rotation=Get WheelSpinBR() -> (WheelSpinBR:Rotator), Scale="1,1,1") {rotationMode=BMM_Replace, rotationSpace=BCS_BoneSpace, alphaBoolBlend {blendOption=Linear}, boneToModify {boneName=wheel_tilted_bk_r}}
   n11 = "Transform (Modify) Bone - Bone: wheel_tilted_bk_l"(ComponentPose=@n10, Alpha=1, Translation="0,0,0", Rotation=Get WheelSpinBL() -> (WheelSpinBL:Rotator), Scale="1,1,1") {rotationMode=BMM_Replace, rotationSpace=BCS_BoneSpace, alphaBoolBlend {blendOption=Linear}, boneToModify {boneName=wheel_tilted_bk_l}}
   n12 = "Transform (Modify) Bone - Bone: steering_wheel"(ComponentPose=@n11, Alpha=1, Translation="0,0,0", Rotation=Get SteeringRotation() -> (SteeringRotation:Rotator), Scale="1,1,1") {rotationMode=BMM_Replace, rotationSpace=BCS_BoneSpace, alphaBoolBlend {blendOption=Linear}, boneToModify {boneName=steering_wheel}}
   n13 = RigidBody(ComponentPose=@n12, Alpha=Get triggerTailgate() -> (triggerTailgate:double)) {
    bOverrideWorldGravity=true
    evaluationResetTime=-1
    overridePhysicsAsset=&/Pickup_Truck_Veh_Dyn_dmg_tailgate
    alphaBoolBlend {blendOption=Linear}
    baseBoneRef {boneName=body}
    overrideWorldGravity {z=-980}
    simSpaceSettings {
     maxAngularAcceleration=2000
     maxAngularVelocity=2000
     maxLinearAcceleration=2000
     maxLinearVelocity=2000
     worldAlpha=1
     externalLinearDragV=<1 1 1>
    }
   }
   n14 = RigidBody(ComponentPose=@n13, Alpha=Get TriggerMirrorL() -> (TriggerMirrorL:double)) {
    bOverrideWorldGravity=true
    evaluationResetTime=-1
    overridePhysicsAsset=&/Pickup_Truck_Veh_Dyn_dmg_mirror_l
    alphaBoolBlend {blendOption=Linear}
    baseBoneRef {boneName=body}
    overrideWorldGravity {z=-980}
    simSpaceSettings {
     maxAngularAcceleration=2000
     maxAngularVelocity=2000
     maxLinearAcceleration=2000
     maxLinearVelocity=2000
     worldAlpha=1
     externalLinearDragV=<1 1 1>
    }
   }
   n15 = RigidBody(ComponentPose=@n14, Alpha=Get TriggerMirrorR() -> (TriggerMirrorR:double)) {
    bOverrideWorldGravity=true
    evaluationResetTime=-1
    overridePhysicsAsset=&/Pickup_Truck_Veh_Dyn_dmg_mirror_r
    alphaBoolBlend {blendOption=Linear}
    baseBoneRef {boneName=body}
    overrideWorldGravity {z=-980}
    simSpaceSettings {
     maxAngularAcceleration=2000
     maxAngularVelocity=2000
     maxLinearAcceleration=2000
     maxLinearVelocity=2000
     worldAlpha=1
     externalLinearDragV=<1 1 1>
    }
   }
   n16 = Get hasTireMod() -> (hasTireMod:bool)
   n17 = "Transform (Modify) Bone - Bone: wheel_steering_bk_l"!(ComponentPose=@n15, bAlphaBoolEnabled:bool=@n16, Translation="0,-15,0", Rotation="0,0,0", Scale="1,1,1") -> (Pose) {
    alphaInputType=Bool
    rotationSpace=BCS_BoneSpace
    translationMode=BMM_Additive
    translationSpace=BCS_BoneSpace
    alphaBoolBlend {blendOption=Linear}
    boneToModify {boneName=wheel_steering_bk_l}
    translation {y=-15}
   }
   n18 = "Transform (Modify) Bone - Bone: wheel_steering_bk_r"!(ComponentPose=@n17, bAlphaBoolEnabled:bool=@n16, Translation="0,15,0", Rotation="0,0,0", Scale="1,1,1") -> (Pose) {
    alphaInputType=Bool
    rotationSpace=BCS_BoneSpace
    translationMode=BMM_Additive
    translationSpace=BCS_BoneSpace
    alphaBoolBlend {blendOption=Linear}
    boneToModify {boneName=wheel_steering_bk_r}
    translation {y=15}
   }
   n19 = "Transform (Modify) Bone - Bone: steering_knuckle_fr_l"!(ComponentPose=@n18, bAlphaBoolEnabled:bool=@n16, Translation="0,-15,0", Rotation="0,0,0", Scale="1,1,1") -> (Pose) {
    alphaInputType=Bool
    rotationSpace=BCS_BoneSpace
    translationMode=BMM_Additive
    translationSpace=BCS_BoneSpace
    alphaBoolBlend {blendOption=Linear}
    boneToModify {boneName=steering_knuckle_fr_l}
    translation {y=-15}
   }
   n20 = "Transform (Modify) Bone - Bone: steering_knuckle_fr_r"!(ComponentPose=@n19, bAlphaBoolEnabled:bool=@n16, Translation="0,15,0", Rotation="0,0,0", Scale="1,1,1") -> (Pose) {
    alphaInputType=Bool
    rotationSpace=BCS_BoneSpace
    translationMode=BMM_Additive
    translationSpace=BCS_BoneSpace
    alphaBoolBlend {blendOption=Linear}
    boneToModify {boneName=steering_knuckle_fr_r}
    translation {y=15}
   }
   Output Pose(Result=Component To Local(ComponentPose=@n20)) {layerGroup=None}
  ]}
 EventGraph {nodes [
   Event BlueprintUpdateAnimation()
   > n1 = Sequence!(execute) -> (then_0, then_1, then_2:Exec, then_3:Exec)
   n2 = Get Vehicle() -> (Vehicle:Fort Athena SKVehicle Object Reference)
   n3 = float / float(
    A=GetHealth!(self:Fort Vehicle Interface Interface=@n2) -> (ReturnValue:float)
    B=GetMaxHealth!(self:Fort Vehicle Interface Interface=@n2) -> (ReturnValue:float)
   ) {is=title K2Node_CallFunction}
   > n4 = Branch(Condition="float <= float"!(A:double=@n3, B:double=0.55) -> (ReturnValue:bool) {is=title K2Node_CallFunction})
   > Set triggerTailgate(triggerTailgate:double=1) -> (Output_Get:double)
   Set triggerTailgate(execute=@n4.else, triggerTailgate=0)
   n5 = Branch(execute=@n1.then_1, Condition="float <= float"(A=@n3, B=0.25))
   > Set TriggerMirrorL(TriggerMirrorL:double=1) -> (Output_Get:double)
   Set TriggerMirrorL(execute=@n5.else, TriggerMirrorL=0)
   n6 = Branch(execute=@n1.then_2, Condition="float <= float"(A=@n3, B=0.75))
   > Set TriggerMirrorR(TriggerMirrorR:double=1) -> (Output_Get:double)
   Set TriggerMirrorR(execute=@n6.else, TriggerMirrorR=0)
   n7 = Cast To FortDagwoodVehicle!(execute=@n1.then_3, Object=@n2) -> (then, CastFailed, AsFort Dagwood Vehicle:Fort Dagwood Vehicle Object Reference) {is=title K2Node_DynamicCast}
   > Set hasTireMod(hasTireMod:bool=HasTireMod!(self:Fort Vehicle Interface Interface=@n7) -> (ReturnValue:bool)) -> (Output_Get:bool)
  ]}
]
variables [
 hasTireMod {type=boolean}
 TriggerMirrorL {type=number}
 TriggerMirrorR {type=number}
 triggerTailgate {type=number}
]
