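# Physics asset for the Valet pickup truck, named as its pushout variant, giving a simplified collision shape for the vehicle.
# A single root body holds three boxes: a near-upright tailgate, a long low bed floor that extends forward beneath the cab, and a cab roof slab.
# It has no constraints, and no boxes cover the wheels or side walls.
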PhysicsAsset /Valet/Pickup_Truck/Meshes/Pickup_Truck_Veh_Physics_Pushout
bodies {root [
  {
   shapeType=Box
   extentX=78.375
   extentY=214.9375
   extentZ=21.5
   shapeName=Phys_Tailgate
   center=<-206.0569 0.0024 90.1865>
   rotation=<rot 87.1873 0 0>
  }
  {
   shapeType=Box
   extentX=334.5
   extentY=219.5625
   extentZ=35.4375
   shapeName=Phys_Bed_Floor
   center=<-47.6684 -0.0002 64.4434>
   rotation=<rot 0.0001 0 0>
  }
  {shapeType=Box, extentX=77, extentY=160.75, extentZ=14.875, shapeName=CabRoof, center=<55.3733 0 176.8986>, rotation=<rot 0 0 -0.0003>}
 ]}
constraints []
